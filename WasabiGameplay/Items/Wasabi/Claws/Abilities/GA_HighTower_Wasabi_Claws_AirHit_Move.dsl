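; Gameplay ability for the Wasabi claws' air-hit dash, propelling the pawn toward the closest selected target, or a set distance forward without one.
; Root motion forces drive travel, with knockback immunity and overlapping pawn collision until the ability ends.
; Montage notifies apply damage and end the ability, backed by a failsafe timer and knocked-back tag listener.
; After travel, a ground trace chooses temporary gravity override or a stopped landing.
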(event Custom|BindOnKnockedBack
  ;[1 1 1] Bind On Knocked Back
  (bind _asynctaskproxy (Ability|Tasks|WaitGameplayTagAdd (Variables|Default|GetTag_KnockedBack))
    (:then
      (Variables|Default|SetTask_OnKnockedBack _asynctaskproxy)
      (Development|PrintString "Bound"))
    (:Added
      (CallFunction|FinishAbility)
      (Development|PrintString "Triggered" true true "(R=0.096780,G=1.000000,B=0.000000,A=1.000000)" 5.0))))

(event Custom|FailsafeTimer
  ;[1 1 1] On Ability Activated, try to find a target actor to move to. If no actor found, move a set distance, then do Finished Moving
  ;;[1 1 1] Failsafe - Force out if we get stuck
  (CallFunction|FinishAbility))

(event Ability|EventActivateAbilityFromEvent
  (params (EventData :type Gameplay%20Event%20Data%20Structure%20%28by%20ref%29))
  (bind (_outlocation _outrotation) (Actor|GetActorEyesViewPoint (Variables|Default|GetPawn)))
  (bind _pawn (Variables|Default|GetPawn))
  (bind _capsulecomponent (Variables|/Script/Engine.Character|GetCapsuleComponent _pawn))
  (bind _returnvalue (Components|Capsule|GetScaledCapsuleRadius _capsulecomponent))
  (bind (_returnvalue_x _returnvalue_y _returnvalue_z) (Math|Vector|BreakVector (* (Math|Vector|SelectVector (Transformation|GetActorForwardVector (Variables|Default|GetPawn)) (Math|Vector|GetForwardVector _outrotation) (Variables|Balance|GetConstrainNoTargetMovementToHorizontalOnly)) (ScalableFloat|GetValueatLevel (Variables|Balance|GetSpeedAfterTravelNoTarget)))))
  ;[1 1 1] On Ability Activated, try to find a target actor to move to. If no actor found, move a set distance, then do Finished Moving
  (|CallParentFunction EventData)
  (if (not (Variables|Default|GetAbilityActive))
    ;;[1 1 1] Ability Setup
    (CallFunction|InitAbility)
    (Input|SetRotatePawntoCamera (Fortnite|GetFortPlayerControllerfromActor (Variables|Default|GetPawn)) false)
    (CallFunction|SetKnockbackImmunity true)
    (CallFunction|SetPawnCollision (Variables|Default|GetPawn))
    (Ability|ApplyGameplayEffectSpecToOwner (Ability|MakeOutgoingGameplayEffectSpec (Variables|Default|GetGE_PreventJumpOnActivate)))
    (bind _returnvalue_2 (Ability|ApplyGameplayEffectSpecToOwner (Ability|MakeOutgoingGameplayEffectSpec (Variables|Default|GetGE_ActivePower))))
    (Variables|Default|SetHandle_GE_ActivePower _returnvalue_2)
    (Ability|CommitAbility)
    (bind (_asynctaskproxy _notifytag) (Ability|Tasks|PlayMontageAndWaitNotify "None" (Variables|Default|GetMontage) 1.0 (Variables|Default|GetMontageStartSection))
      (:then
        (bind (_asynctaskproxy_2 _targetdata) (Ability|Tasks|DoTargetSelection "None" (Variables|Default|GetTag_FindTarget) :!pinned true)
          (:then
            ;;[1 1 1] Failsafe - Force out if we get stuck
            (bind _returnvalue_3 (Utilities|Time|SetTimerbyEvent (delegate Custom|FailsafeTimer) (Variables|Balance|GetFailsafeAbilityDuration)))
            (Variables|Default|SetFailsafeTimerHandle _returnvalue_3))
          (:Targeted
            (if (Ability|TargetData|TargetDataHasActor _targetdata)
              ;;[1 1 1] Move to Target Actor if Found
              (Variables|Default|SetTargetActor (Fortnite|GetClosestActorfromArray (Variables|Default|GetPawn) (Ability|TargetData|GetActorsfromTargetData _targetdata)))
              (bind _distance (Class|MissionBlueprintFunctionLibrary|DistanceBetweenTwoVectors (Transformation|GetActorLocation (Variables|Default|GetPawn)) (Transformation|GetActorLocation (Variables|Default|GetTargetActor))))
              (Variables|Default|SetDistanceToTarget _distance)
              (bind _asfort_pawn (Utilities|Casting|CastToFortPawn (Variables|Default|GetTargetActor))
                (:then
                  (Variables|Default|SetActorsCombinedScaledRadiusSize (+ (Components|Capsule|GetScaledCapsuleRadius (Variables|/Script/Engine.Character|GetCapsuleComponent _asfort_pawn)) _returnvalue)))
                (:CastFailed
                  (Variables|Default|SetActorsCombinedScaledRadiusSize _returnvalue)))
              (bind _asynctaskproxy_3 (Ability|Tasks|ApplyRootMotionMoveToActorForce "None" (Variables|Default|GetTargetActor) (Math|Vector|MakeVector (Math|Float|Max(Float) (ScalableFloat|GetValueatLevel (Variables|Balance|GetDesiredLandingDistanceFromTarget)) (Variables|Default|GetActorsCombinedScaledRadiusSize))) "AlignFromTargetToSource" (Math|Float|Max(Float) (Math|Float|Lerp (ScalableFloat|GetValueatLevel (Variables|Balance|GetMinTravelDuration)) (ScalableFloat|GetValueatLevel (Variables|Balance|GetMaxTravelDuration)) (Math|Float|Clamp(Float) (/ (Variables|Default|GetDistanceToTarget) (ScalableFloat|GetValueatLevel (Variables|Balance|GetTravelDistanceNoTarget)))))) 0 0 false "MOVE_None" false (Variables|Balance|GetPathOffsetCurve) 0 "SetVelocity" "0, 0, 0" 0.0 true)
                (:then
                  (Variables|Default|SetRootMotionMoveToActorTask _asynctaskproxy_3))
                (:OnFinished
                  (CallFunction|FinishedMoving)))
              (else
                ;;[1 1 1] Move to Location if no Target Actor found
                (bind _asynctaskproxy_4 (Ability|Tasks|ApplyRootMotionMoveToForce "None" (+ (Transformation|GetActorLocation (Variables|Default|GetPawn)) (* (Math|Vector|SelectVector (Transformation|GetActorForwardVector (Variables|Default|GetPawn)) (Math|Vector|GetForwardVector _outrotation) (Variables|Balance|GetConstrainNoTargetMovementToHorizontalOnly)) (ScalableFloat|GetValueatLevel (Variables|Balance|GetTravelDistanceNoTarget)))) (ScalableFloat|GetValueatLevel (Variables|Balance|GetMaxTravelDuration)) false "MOVE_None" false (Variables|Balance|GetPathOffsetCurve) "SetVelocity" (Math|Vector|MakeVector _returnvalue_x _returnvalue_y (Math|Float|Clamp(Float) _returnvalue_z (* (ScalableFloat|GetValueatLevel (Variables|Balance|GetZVelocityClampAfterTravelNoTarget)) -1.0) (ScalableFloat|GetValueatLevel (Variables|Balance|GetZVelocityClampAfterTravelNoTarget)))))
                  (:then
                    (Variables|Default|SetRootMotionMoveToForceTask _asynctaskproxy_4))
                  (:OnTimedOut
                    _)
                  (:OnTimedOutAndDestinationReached
                    _))
                (CallFunction|FinishedMoving))))
          (:Cancelled
            (CallFunction|FinishAbility))) ; MoveTo Target Selection
      )
      (:OnCompleted
        _)
      (:OnBlendOut
        _)
      (:OnInterrupted
        _)
      (:OnCancelled
        _)
      (:OnNotifyBegin
        (if (GameplayTags|Equal(GameplayTag) _notifytag (Variables|Default|GetTag_DamageContainer))
          (bind (_asynctaskproxy_5 _targetdata_2 _applicationtag) (Ability|Tasks|DoTargetSelection "None" (Variables|Default|GetTag_DamageContainer) :!pinned true)
            (:Targeted
              (Ability|ApplyGameplayEffectContainer _targetdata_2 _applicationtag))) ; Damage Target Selection
          (elif (GameplayTags|Equal(GameplayTag) _notifytag (Variables|Default|GetTag_EndAbility))
            (CallFunction|FinishAbility) ; Finish Ability from Anim Notify
          ))))
    (CallFunction|FinishAbility)))

(event Custom|FinishedMoving
  ;[1 1 1] Finished Moving - End move tasks. Check height above ground and add low gravity if at sufficient height, else stop the character - then Finish Ability
  (GameplayTasks|EndTask (Variables|Default|GetRootMotionMoveToForceTask))
  (GameplayTasks|EndTask (Variables|Default|GetRootMotionMoveToActorTask))
  (if (not (Math|Boolean|ANDBoolean (Health|IsDBNO (Variables|Default|GetPawn)) (Utilities|IsValid (Variables|Default|GetPawn) :!class /Script/Engine.KismetSystemLibrary) (FortPawn|IsDead (Variables|Default|GetPawn))))
    (bind (_outhit _returnvalue) (Collision|LineTraceByChannel (Transformation|GetActorLocation (Variables|Default|GetPawn)) (Math|Vector|vector-vector (Transformation|GetActorLocation (Variables|Default|GetPawn)) (Math|Vector|MakeVector 0.0 0.0 (Variables|Balance|GetMaxHeightToBeConsideredOnGround))) "TraceTypeQuery10")) ; Check to see if we are on or near ground
    (if (and (not _returnvalue) (AI|Components|NavMovement|IsFalling (Variables|/Script/Engine.Character|GetCharacterMovement (Variables|Default|GetPawn))))
      (bind _returnvalue_2 (Ability|GameplayEffect|AssignTagSetbyCallerMagnitude (Ability|MakeOutgoingGameplayEffectSpec (Variables|Default|GetGE_SetGravityAfterFinishTravel)) "(TagName=\"SetByCaller.Duration\")" (ScalableFloat|GetValueatLevel (Variables|Balance|GetGravityDurationAfterFinishTravel))))
      (bind _returnvalue_3 (Ability|GameplayEffect|AssignTagSetbyCallerMagnitude _returnvalue_2 "(TagName=\"SetByCaller.Magnitude\")" (ScalableFloat|GetValueatLevel (Variables|Balance|GetGravityOverrideAfterFinishTravel))))
      (Ability|ApplyGameplayEffectSpecToOwner _returnvalue_3)
      (Ability|Animation|MontageJumptoSection "HitInAir")
      (Development|PrintString "Hit In Air")
      (else
        (Variables|/Script/Engine.MovementComponent|SetVelocity "0, 0, 0" (Variables|/Script/Engine.Character|GetCharacterMovement (Variables|Default|GetPawn)))
        (Ability|ApplyGameplayEffectSpecToOwner (Ability|MakeOutgoingGameplayEffectSpec (Variables|Default|GetGE_Landed_MoveSpeed)))
        (Ability|Animation|MontageJumptoSection "Land")
        (Development|PrintString "Landed")))))

(event Custom|FinishAbility
  ;[1 1 1] Finish Ability
  (Ability|EndAbility))

(event Ability|EventOnEndAbility
  (params (bWasCancelled :type Boolean))
  ;[1 1 1] On End Ability - also gets called when unequipping the weapon (removing the ability)
  (|CallParentFunction bWasCancelled)
  (Development|PrintString "AirDash Ability End" true true "(R=0.000000,G=1.000000,B=0.008628,A=1.000000)")
  (Utilities|Time|ClearandInvalidateTimerbyHandle (Variables|Default|GetFailsafeTimerHandle))
  (GameplayTasks|EndTask (Variables|Default|GetRootMotionMoveToForceTask))
  (GameplayTasks|EndTask (Variables|Default|GetRootMotionMoveToActorTask))
  (Input|SetRotatePawntoCamera (Fortnite|GetFortPlayerControllerfromActor (Variables|Default|GetPawn)) true)
  (CallFunction|SetKnockbackImmunity false)
  (CallFunction|SetPawnCollision (Variables|Default|GetPawn) true)
  (Pawn|Character|ResetFallingHeight (Variables|Default|GetPawn))
  (Variables|Default|SetAbilityActive false)
  (Ability|RemoveGameplayEffectFromOwnerWithHandle (Variables|Default|GetHandle_GE_ActivePower))
  (GameplayTasks|EndTask (Variables|Default|GetTask_OnKnockedBack)))

(fn InitAbility
  (bind _returnvalue (Ability|GetAvatarActorfromActorInfo))
  (Utilities|IsValid _returnvalue
    (:Is%20Valid
      (bind _asfort_pawn (Utilities|Casting|CastToFortPawn _returnvalue)
        (:then
          (Variables|Default|SetPawn _asfort_pawn)
          (Variables|Default|SetAbilityActive true)
          (CallFunction|BindonKnockedBack))
        (:CastFailed
          _)))
    (:Is%20Not%20Valid
      _))
  (Ability|EndAbility))

(fn SetKnockbackImmunity
  (params (ImmunityOn :type Boolean))
  (if ImmunityOn
    ;[1 1 1] Apply Knockback Immunity
    (bind _returnvalue (Ability|ApplyGameplayEffectSpecToOwner (Ability|MakeOutgoingGameplayEffectSpec (Variables|Default|GetGE_KnockbackImmunity))))
    (Variables|Default|SetHandle_KnockbackImmunity _returnvalue)
    (else
      ;[1 1 1] Remove Knockback Immunity
      (Ability|RemoveGameplayEffectFromOwnerWithHandle (Variables|Default|GetHandle_KnockbackImmunity)))))

(fn SetPawnCollision
  (params (FortPawn :type Fort%20Pawn%20Object%20Reference) (CollisionOn :type Boolean))
  (bind _savedcollisionchannel (Variables|Default|GetSavedCollisionChannel))
  (bind _capsulecomponent (Variables|/Script/Engine.Character|GetCapsuleComponent FortPawn))
  (Utilities|IsValid FortPawn
    (:Is%20Valid
      (if CollisionOn
        ;[1 1 1] Set Pawn collision back to blocking if the saved collision is not none
        (if (Utilities|Name|NotEqual(Name) _savedcollisionchannel "None")
          (Collision|SetCollisionProfileName _capsulecomponent _savedcollisionchannel true :!class /Script/Engine.PrimitiveComponent)
          (Collision|SetCollisionResponsetoChannel _capsulecomponent "ECC_Pawn" "ECR_Block")
          (Collision|SetCollisionResponsetoChannel _capsulecomponent "ECC_GameTraceChannel4" "ECR_Block"))
        (else
          ;[1 1 1] Set Pawn collision to overlap and save out collision
          (Variables|Default|SetSavedCollisionChannel (Collision|GetCollisionProfileName _capsulecomponent))
          (Collision|SetCollisionResponsetoChannel _capsulecomponent "ECC_Pawn" "ECR_Overlap")
          (Collision|SetCollisionResponsetoChannel _capsulecomponent "ECC_GameTraceChannel4" "ECR_Overlap"))))))
