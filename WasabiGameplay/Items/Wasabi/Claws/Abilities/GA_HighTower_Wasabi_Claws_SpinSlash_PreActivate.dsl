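; Gameplay ability that pre-activates the claws' spin slash attack, judging by its High Tower Wasabi Claws naming.
; On event-triggered activation it waits a short delay, then sends the SpinSlash activate gameplay event tag to its avatar actor.
; The event data from the trigger is not forwarded, and the graph shown never explicitly ends the ability.
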(event Ability|EventActivateAbilityFromEvent
  (params (EventData :type Gameplay%20Event%20Data%20Structure%20%28by%20ref%29))
  (Utilities|FlowControl|Delay 0.1)
  (Ability|SendGameplayEventtoActor (Ability|GetAvatarActorfromActorInfo) (Variables|Default|GetTag_SpinSlash_Activate)))
