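; Gameplay ability for the Wasabi claws Mantis attack, extending its parent with zipline and event handling.
; On activation while ziplining, it sends a zipline exit event to the avatar and flags a Spin Slash pre-activate event, which it sends when the ability ends.
; It also re-sends animation notify tags as gameplay events, which trigger Air Attack gap closing in a passive weapon ability.
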(event Ability|EventActivateAbility
  ;[1 1 1] On Activate Ability - Handle Ziplining
  (bind _bdidactorhavetags (Class|AthenaQuestFunctionLibrary|AthenaCheckActorHasTags (Ability|GetAvatarActorfromActorInfo) (Variables|Default|GetTag_Ziplining)))
  (if _bdidactorhavetags
    (Ability|SendGameplayEventtoActor (Ability|GetAvatarActorfromActorInfo) (Variables|Default|GetTag_ZiplineExit))
    (Variables|Default|SetSendSpinSlashEvent true)
    (else
      _))
  (|CallParentFunction))

(event Ability|EventOnGenericNotifyFromAnimation
  (params (NotifyTag :type Gameplay%20Tag%20Structure))
  ;[1 1 1] Pass on Generic Notify tags as Gameplay Events (used to trigger Air Attack gap closing which is in a passive ability on the weapon)
  (|CallParentFunction NotifyTag)
  (Ability|SendGameplayEventtoActor (Ability|GetAvatarActorfromActorInfo) NotifyTag))

(event Ability|EventOnEndAbility
  (params (bWasCancelled :type Boolean))
  ;[1 1 1] On Ability End - Handle Ziplining events
  (if (Variables|Default|GetSendSpinSlashEvent)
    (Variables|Default|SetSendSpinSlashEvent false)
    (Ability|SendGameplayEventtoActor (Ability|GetAvatarActorfromActorInfo) (Variables|Default|GetTag_SpinSlash_PreActivate))
    (else
      _))
  (|CallParentFunction bWasCancelled))
