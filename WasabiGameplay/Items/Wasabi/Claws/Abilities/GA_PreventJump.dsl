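; GA_PreventJump is a gameplay ability for the Wasabi Claws item, named for preventing jumps.
; Activation prints a "COOLDOWN STARTED" debug string and ending prints "COOLDOWN ENDED", regardless of cancellation.
; The failed passive activation event is empty, and the graph contains no jump-blocking logic or call to end the ability.
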(event Ability|EventActivateAbility
  (Development|PrintString "COOLDOWN STARTED"))

(event Ability|EventOnEndAbility
  (params (bWasCancelled :type Boolean))
  (Development|PrintString "COOLDOWN ENDED"))

(event Ability|EventFailedToActivatePassiveAbility
  (params (ActorInfo :type Gameplay%20Ability%20Actor%20Info%20Structure)))
